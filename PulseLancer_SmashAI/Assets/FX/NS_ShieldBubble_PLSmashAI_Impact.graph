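# Niagara system for the impact effect on the PulseLancer SmashAI shield bubble.
# It layers six emitters: a star mesh, a ripple, small sparks, a burst wave, a core flare and a light.
# User parameters expose red-pink primary and secondary colours, grid range and size, and impact-hole amount and range.
# Only the SystemState module is listed; emitter renderers and internal logic are not shown.
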NiagaraSystem /PulseLancer_SmashAI/Assets/FX/NS_ShieldBubble_PLSmashAI_Impact
dependencies {usedDataInterfaces [], usedDynamicInputs [], usedModules [&SystemState]}
emitters {
 BurstWave {rendererClasses [], renderers []}
 CoreFlare {rendererClasses [], renderers []}
 Impact_Star_Mesh {rendererClasses [], renderers []}
 Light_LW {rendererClasses [], renderers []}
 Ripple {rendererClasses [], renderers []}
 Sparks_Small {rendererClasses [], renderers []}
}
summary {
 systemName=NS_ShieldBubble_PLSmashAI_Impact
 emitters [
  {emitterName=Impact_Star_Mesh, rendererClasses []}
  {emitterName=Ripple, rendererClasses []}
  {emitterName=Sparks_Small, rendererClasses []}
  {emitterName=BurstWave, rendererClasses []}
  {emitterName=CoreFlare, rendererClasses []}
  {emitterName=Light_LW, rendererClasses []}
 ]
 userVariables [
  User.GridRange {type=NiagaraFloat, defaultValue=50, description=""}
  User.GridSize {type=NiagaraFloat, defaultValue=1, description=""}
  User.ImpactHoleAmount {type=NiagaraInt32, defaultValue=1}
  User.ImpactHoleRange {type=NiagaraFloat, defaultValue=0, description=""}
  User.PrimaryColour {type=LinearColor, description="", defaultValue {struct=LinearColor, value=<rgba 0.4957 0.0364 0.1221 1>}}
  User.SecondaryColour {type=LinearColor, description="", defaultValue {struct=LinearColor, value=<rgba 0.6435 0.056 0.056 1>}}
 ]
}
