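# Niagara system for a shield bubble effect, named for the "valid candy" time-up variant, likely played when that bubble's timer expires.
# Five emitters (CoreFlare, Mesh_Castoff, BurstWave, Sparks_Small, EnergyBurst) suggest a layered burst, though no renderers are listed.
# User parameters expose primary and secondary colours, defaulting to pink-red tones, plus an integer Stage.
# SystemState is the only module listed, so the file shows no emitter behaviour.
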NiagaraSystem /PulseLancer_SmashAI/Assets/FX/NS_ShieldBubble_ValidCandy_TimesUp
dependencies {usedDataInterfaces [], usedDynamicInputs [], usedModules [&SystemState]}
emitters {
 BurstWave {rendererClasses [], renderers []}
 CoreFlare {rendererClasses [], renderers []}
 EnergyBurst {rendererClasses [], renderers []}
 Mesh_Castoff {rendererClasses [], renderers []}
 Sparks_Small {rendererClasses [], renderers []}
}
summary {
 systemName=NS_ShieldBubble_ValidCandy_TimesUp
 emitters [
  {emitterName=CoreFlare, rendererClasses []}
  {emitterName=Mesh_Castoff, rendererClasses []}
  {emitterName=BurstWave, rendererClasses []}
  {emitterName=Sparks_Small, rendererClasses []}
  {emitterName=EnergyBurst, rendererClasses []}
 ]
 userVariables [
  User.PrimaryColour {type=LinearColor, description="", defaultValue {struct=LinearColor, value=<rgba 0.4957 0.0364 0.1221 1>}}
  User.SecondaryColour {type=LinearColor, description="", defaultValue {struct=LinearColor, value=<rgba 0.6435 0.056 0.056 1>}}
  User.Stage {type=NiagaraInt32, defaultValue=2}
 ]
}
