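# Material function that supplies the texture-driven surface attributes for the PulseLancer Deimos enemy as one MakeMaterialAttributes output.
# Parameterized Diffuse and Normal samples feed base color and normal, while the M Mask supplies AO and opacity (blue) and a scalar sets opacity mask.
# The S Mask's green Metalic channel drives metallic, specular and roughness alike, leaving its Specular and Roughness channels unused.
# Four Husk Deimos texture objects sit unconnected.
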MaterialFunction /PulseLancer_SmashAI/Assets/Materials/MF_DeimosEnemy_Pulselancer_Textures
graph {nodes [
  n1 = TextureSampleParameter2D() {
   group=Deimos
   parameterName=Diffuse
   samplerType=SAMPLERTYPE_Masks
   sortPriority=0
   texture=/PulseLancer_SmashAI/Assets/Textures/TA/T_PulseLancer_Proto_D
  }
  n2 = TextureSampleParameter2D() {
   group=Deimos
   parameterName=S Mask
   samplerSource=SSM_Wrap_WorldGroupSettings
   sortPriority=10
   texture=&T_BlacksRGB
   channelNames {a="", b=Roughness, g=Metalic, r=Specular}
  }
  n3 = TextureSampleParameter2D() {
   group=Deimos
   parameterName=Normal
   samplerSource=SSM_Wrap_WorldGroupSettings
   samplerType=SAMPLERTYPE_Normal
   sortPriority=10
   texture=&FlatNormalBlue
  }
  n4 = TextureSampleParameter2D() {
   group=Deimos
   parameterName=M Mask
   samplerSource=SSM_Wrap_WorldGroupSettings
   samplerType=SAMPLERTYPE_Masks
   sortPriority=10
   texture=&T_WhiteMask
   channelNames {a="", b="", g="", r=AO}
  }
  FunctionOutput(None=MakeMaterialAttributes!(
    BaseColor=@n1.RGB
    Metallic=@n2.Metalic
    Specular=@n2.Metalic
    Roughness=@n2.Metalic
    Anisotropy
    EmissiveColor
    Opacity=@n4.B
    OpacityMask=ScalarParameter() {defaultValue=1, group=Base Material, parameterName=Opacity Mask}
    Normal=@n3.RGB
    Tangent
    WorldPositionOffset
    SubsurfaceColor
    ClearCoat
    ClearCoatRoughness
    AmbientOcclusion=@n4.AO
    Refraction
    CustomizedUVs_0
    CustomizedUVs_1
    CustomizedUVs_2
    CustomizedUVs_3
    CustomizedUVs_4
    CustomizedUVs_5
    CustomizedUVs_6
    CustomizedUVs_7
    PixelDepthOffset
    ShadingModel
    Displacement
   )) {outputName=MakeMaterialAttributes}
  TextureObject() {samplerType=SAMPLERTYPE_Masks, texture=/Game/Characters/Enemies/Husk_Deimos/Textures/T_Husk_Deimos_01_Base_S}
  TextureObject() {texture=/Game/Characters/Enemies/Husk_Deimos/Textures/T_Husk_Deimos_01_Base_D}
  TextureObject() {samplerType=SAMPLERTYPE_Normal, texture=/Game/Characters/Enemies/Husk_Deimos/Textures/T_Husk_Deimos_01_Base_N}
  TextureObject() {samplerType=SAMPLERTYPE_Masks, texture=/Game/Characters/Enemies/Husk_Deimos/Textures/T_Husk_Deimos_01_Base_M}
 ]}
