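# Look-dev material function for PulseLancer fiends that overrides base color, metallic, roughness and emissive on incoming material attributes.
# Base color uses either prototype height gradients with demon markings, or RGB-masked tint variants driven by custom primitive data.
# Demon-mark SDFs use pre-skinned position and normals for single-sample triplanar projection with explicit derivatives.
# Emissive combines marks, FX mask, rim light and switchable exploding pulse and ranged-hand glows, compensated for eye adaptation.
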MaterialFunction /PulseLancer_SmashAI/Assets/Materials/MF_FiendLookDev_PulseLancer
graph {nodes [
  ComponentMask(None=TextureCoordinate()) {r=true}
  Constant3Vector() {constant=<rgba 0 0 1 1>}
  Constant() {r=1}
  n1 = GetMaterialAttributes(None=FunctionInput() {inputName=Result, inputType=FunctionInput_MaterialAttributes}) {attributeGetTypes [
    #1
    #2
    #3
    #4
    #5
    #6
    #7
   ]}
  n2 = ScalarParameter() {group=PulseLancer_Prototype, parameterName=Proto_Bias_Metallic}
  n3 = NamedRerouteUsage()
  n4 = NamedRerouteDeclaration(None=@n1.Roughness) {name=IN_Roughness, nodeColor=<rgba 0 0.2969 1 1>}
  n5 = ScalarParameter() {group=PulseLancer_Prototype, parameterName=Proto_Bias_Roughness}
  n6 = SetMaterialAttributes!(
   MaterialAttributes=@n1.MaterialAttributes
   BaseColor=NamedRerouteUsage()
   Metallic=StaticSwitch(True=Add(A=@n1.Metallic, B=@n2), False=@n1.Metallic, Value=@n3)
   Specular
   Roughness=StaticSwitch(True=Add(A=@n4, B=@n5), False=@n4, Value=@n3)
   Anisotropy
   EmissiveColor=NamedRerouteUsage()
   Opacity
   OpacityMask
   Normal
   AmbientOcclusion
  ) {attributeSetTypes [
    #1
    #2
    #3
    #4
    #5
    #6
    #7
    #8
    #9
    #10
   ]}
  FunctionOutput(None=@n6) {outputName=SetMaterialAttributes}
  n7 = NamedRerouteUsage()
  n8 = Divide(
   A=Subtract(
    A=ComponentMask(None=@n7) {b=true}
    B=ScalarParameter() {defaultValue=70, group=Deimos_Emissive, parameterName=Emissive_Hands_Gradient_Position_Z}
   ) {constB=60}
   B=ScalarParameter() {defaultValue=70, group=Deimos_Emissive, parameterName=Emissive_Hands_TransitionFalloff}
  )
  n9 = Subtract(A=Multiply(A=NamedRerouteUsage()) {constB=2})
  n10 = ScalarParameter() {defaultValue=0.3, desc=0.17, group=Deimos_Emissive, parameterName=Emissive_Hands_Noise}
  n11 = Saturate(None=Subtract(A=OneMinus(None=@n8), B=Multiply(A=@n9, B=@n10)))
  n12 = Constant() {r=4}
  n13 = Multiply(A=Power(Base=NamedRerouteUsage(), Exp=Constant() {r=8}), B=@n12)
  n14 = Power(Base=Divide(A=Add(A=@n11, B=@n13), B=Add(A=@n12)), Exp=Constant() {r=3})
  n15 = ScalarParameter() {defaultValue=1.5, group=Deimos_Emissive, parameterName=Emissive_Hands_Intensity}
  n16 = ScalarParameter() {defaultValue=20, group=Deimos_Emissive, parameterName=Emissive_Hands_Fresnel}
  n17 = NamedRerouteUsage()
  n18 = NamedRerouteUsage()
  Multiply(A=Multiply(A=Multiply(A=@n14, B=@n15), B=LinearInterpolate(B=@n16, Alpha=@n17) {constA=1}), B=Multiply(A=@n18, B=@n18))
  n19 = NamedRerouteUsage()
  n20 = ScalarParameter() {defaultValue=2, group=PulseLancer_Prototype, parameterName=PulseLancer_Markings_BordersNoiseB}
  n21 = NamedRerouteUsage()
  n22 = Add(A=@n21, B=Subtract(A=@n21, B=Multiply(A=@n19, B=@n20)))
  n23 = Add(A=OneMinus(None=@n20))
  Multiply(A=MaterialFunctionCall!(
    X range Y range=Constant4Vector() {constant=<rgba -2 2 -2 2>}.RGBA
    Gradient Height
    ShowGrid?
    Thickness
    Color
    "f(x)"=Divide(A=@n22, B=@n23)
   ) {materialFunction=/Engine/Functions/Engine_MaterialFunctions01/Debug/PlotFunctionOnGraph}."f(x) = y")
  n24 = Divide(
   A=NamedRerouteUsage()
   B=ScalarParameter() {defaultValue=35, group=PulseLancer_Prototype, parameterName=PulseLancer_DemonMarks_Scale, sortPriority=15}
  )
  n25 = TextureObject() {samplerType=SAMPLERTYPE_LinearGrayscale, texture=/PulseLancer_SmashAI/Assets/Textures/TA/T_DemonMarkings}
  n26 = NamedRerouteUsage()
  n27 = NamedRerouteUsage()
  n28 = NamedRerouteDeclaration(None=Subtract(
    A=TextureSample!(UVs=@n24, Tex=@n25, "DDX(UVs)"=@n26, "DDY(UVs)"=@n27, Apply View MipBias) {mipValueMode=TMVM_Derivative, samplerType=SAMPLERTYPE_LinearGrayscale, texture=/Engine/EngineMaterials/DefaultDiffuse}.R
    B=NamedRerouteUsage()
   )) {name=M_Proto_DemonMarksSDF_Fullbody, nodeColor=<rgba 0.3359 1 0 1>}
  n29 = NamedRerouteUsage()
  NamedRerouteDeclaration(None=StaticSwitch(True=Subtract(A=@n28, B=@n29), False=NamedRerouteUsage(), Value=NamedRerouteUsage())) {name=M_DemonMarksSDF, nodeColor=<rgba 1 0.3516 0 1>}
  n30 = TextureSampleParameter2D() {
   desc="R = Emissive Demon Marks SDF\r\nG = Emissive Pulsing Mask (Explosive RibBones)\r\nB = Emissive Constant Mask "
   group=Deimos_Emissive
   parameterName=FX Mask
   samplerSource=SSM_Wrap_WorldGroupSettings
   sortPriority=0
   texture=&T_BlacksRGB
  }
  NamedRerouteDeclaration(None=@n30.B) {name=M_FXTex_EmissiveMask, nodeColor=<rgba 0 0 1 1>}
  NamedRerouteDeclaration(None=@n30.G) {name=M_FXTex_MaskOutEmissive, nodeColor=<rgba 1 0 0.5859 1>}
  NamedRerouteDeclaration(None=@n30.R) {name=M_FXTex_DemonMarks, nodeColor=<rgba 0 0.9062 1 1>}
  n31 = VectorParameter() {group=Deimos_Emissive, parameterName=Emissive_Color, sortPriority=1, defaultValue=<rgba 0.8904 0 0.7673 1>}
  NamedRerouteDeclaration(None=@n31.RGB) {name=EmissiveColor, nodeColor=<rgba 1 0 0.875 1>}
  n32 = LocalPosition() {localOrigin=InstancePreSkinning}
  NamedRerouteDeclaration(None=ShaderStageSwitch(PixelShader=VertexInterpolator(VS=@n32.XYZ).PS, VertexShader=@n32.XYZ)) {name=Preskin_LocalPos, nodeColor=<rgba 1 0.021 0 1>}
  n33 = Clamp(None=DotProduct(A=PixelNormalWS(), B=CameraVectorWS()))
  NamedRerouteDeclaration(None=Power(
    Base=NamedRerouteDeclaration(None=@n33) {name=Fresnel, nodeColor=<rgba 0 1 0.4375 1>}
    Exp=ScalarParameter() {defaultValue=16, group=Deimos_Emissive, parameterName=Emissive_Fresnel_Falloff, sortPriority=1}
   )) {name=M_Fresnel_Core, nodeColor=<rgba 0.1953 0 1 1>}
  n34 = NamedRerouteUsage()
  n35 = ScalarParameter() {defaultValue=0.3, group=Deimos_Emissive, parameterName=Emissive_Pulsing_DemonMarksThickness}
  n36 = Multiply(
   A=Power(
    Base=Add(A=@n34, B=@n35) {constB=0.3}
    Exp=ScalarParameter() {defaultValue=5, group=Deimos_Emissive, parameterName=Emissive_Pulsing_DemonMarks_Power}
   )
   B=ScalarParameter() {defaultValue=2, group=Deimos_Emissive, parameterName=Emissive_Pulsing_DemonMarksMult}
  )
  n37 = Time()
  n38 = ScalarParameter() {defaultValue=3.6, group=Deimos_Emissive, parameterName=Emissive_Pulsing_PulseSpeed}
  n39 = Multiply(
   A=Sine(None=Divide(A=@n37, B=@n38))
   B=ScalarParameter() {defaultValue=0.3, group=Deimos_Emissive, parameterName=Emissive_Pulsing_Pulse Intensity}
  ) {constB=0.1}
  n40 = Multiply(
   A=Multiply(A=NamedRerouteUsage(), B=Add(A=@n39)) {constB=0.5}
   B=ScalarParameter() {defaultValue=2, group=Deimos_Emissive, parameterName="Emissive_Pulsing (Explosive)"}
  )
  n41 = Multiply(A=Max(A=@n40, B=Multiply(A=@n40, B=@n36)), B=NamedRerouteUsage())
  n42 = Multiply(A=@n41, B=ScalarParameter() {defaultValue=0.75, group=Deimos_Emissive, parameterName=Emissive_Pulsing_Pattern_Mult}) {constB=0.5}
  n43 = NamedRerouteUsage()
  n44 = Constant() {r=1}
  n45 = NamedRerouteUsage()
  n46 = NamedRerouteUsage()
  NamedRerouteDeclaration(None=Multiply(A=LinearInterpolate(A=@n41, B=@n42, Alpha=@n43), B=StaticSwitch(True=@n44, False=@n45, Value=@n46))) {name=Emissive_Pulsing, nodeColor=<rgba 0.125 1 0 1>}
  n47 = Subtract(A=Multiply(A=NamedRerouteUsage()) {constB=2})
  n48 = ScalarParameter() {defaultValue=0.3, desc=0.17, group=Deimos_Emissive, parameterName=Emissive_Hands_Noise}
  n49 = NamedRerouteUsage()
  n50 = Multiply(A=@n49, B=Subtract(A=@n49, B=Multiply(A=@n47, B=@n48)))
  n51 = Constant() {r=2.5}
  n52 = Power(
   Base=Add(A=NamedRerouteUsage(), B=ScalarParameter() {defaultValue=0.1, group=Deimos_Emissive, parameterName=Emissive_Hands_DemonMarksThickness}) {constB=0.3}
   Exp=ScalarParameter() {defaultValue=8, group=Deimos_Emissive, parameterName=Emisssive_Hands_DemonMarks_Power, sliderMax=16}
  )
  n53 = ScalarParameter() {defaultValue=2, group=Deimos_Emissive, parameterName=Emisssive_Hands_DemonMarksMult}
  n54 = Add(A=Saturate(None=Power(Base=@n50, Exp=@n51) {constExponent=4}), B=Multiply(A=NamedRerouteUsage(), B=Multiply(A=@n52, B=@n53)))
  n55 = LinearInterpolate(B=ScalarParameter() {defaultValue=20, group=Deimos_Emissive, parameterName=Emissive_Hands_Fresnel}, Alpha=NamedRerouteUsage()) {constA=1}
  NamedRerouteDeclaration(None=Multiply(A=Multiply(A=@n54, B=@n55), B=ScalarParameter() {defaultValue=1.5, group=Deimos_Emissive, parameterName=Emissive_Hands_Intensity})) {name=Glowing_hands, nodeColor=<rgba 1 0 0.4453 1>}
  n56 = TextureCoordinate()
  NamedRerouteDeclaration(None=DDX(Value=@n56)) {name=DDX, nodeColor=<rgba 1 0 0.7031 1>}
  n57 = NamedRerouteUsage()
  n58 = NamedRerouteUsage()
  n59 = ScalarParameter() {defaultValue=5, group=Deimos_Emissive, parameterName=Emissive_DemonMarksSDF_Power}
  n60 = NamedRerouteDeclaration(None=Multiply(A=Power(Base=@n58, Exp=@n59), B=NamedRerouteUsage())) {name=Emissive_DemonMarks, nodeColor=<rgba 0 1 0.6016 1>}
  n61 = ScalarParameter() {defaultValue=1, group=Deimos_Emissive, parameterName=Emissive_DemonMarks}
  n62 = NamedRerouteUsage()
  n63 = NamedRerouteUsage()
  n64 = NamedRerouteUsage()
  n65 = Add(
   A=Multiply(A=Multiply(A=@n60, B=@n61), B=NamedRerouteUsage())
   B=Multiply(
    A=StaticSwitch(True=@n62, False=@n63, Value=@n64)
    B=ScalarParameter() {defaultValue=3, desc=Uses B channel from FX Mask Texture, group=Deimos_Emissive, parameterName=Emissive_Constant}
   )
  )
  n66 = StaticSwitch(True=Add(A=@n65, B=NamedRerouteUsage()), False=@n65, Value=NamedRerouteUsage())
  n67 = Multiply(A=StaticSwitch(True=Add(A=@n66, B=@n57), False=@n66, Value=NamedRerouteUsage()), B=NamedRerouteUsage())
  n68 = NamedRerouteUsage()
  NamedRerouteDeclaration(None=EyeAdaptationInverse(LightValueInput=Max(A=@n67, B=@n68), AlphaInput=Constant() {r=1}).EyeAdaptationInverse) {name=OUT_Emissive, nodeColor=<rgba 1 0 0.9375 1>}
  n69 = VectorParameter() {group=PulseLancer_Prototype, parameterName=PulseLancer_Markings_Color1, defaultValue=<rgba 0.1746 0.1274 0.7682 0.5>}
  n70 = ScalarParameter() {defaultValue=0.4, group=PulseLancer_Prototype, parameterName="PulseLancer_Markings_CutoutValue (0-1)", sliderMax=1}
  n71 = Add(A=@n70, B=Constant() {r=0.1}) {constB=0.15, desc=DemonMarkssoftness}
  n72 = ScalarParameter() {defaultValue=0.1, group=PulseLancer_Prototype, parameterName=PulseLancer_Markings_BordersNoise}
  n73 = Multiply(A=NamedRerouteUsage(), B=@n72)
  n74 = ShadingPathSwitch(Default=Add(A=Subtract(A=@n21, B=@n73), B=Divide(A=@n72)), Mobile=@n21)
  n75 = ScalarParameter() {defaultValue=0.2, group=PulseLancer_Prototype, parameterName=PulseLancer_Markings_Borders, sliderMax=1}
  n76 = LinearInterpolate(
   A=@n69.RGB
   B=Multiply(A=@n69.RGB, B=ScalarParameter() {defaultValue=0.7, group=PulseLancer_Prototype, parameterName=PulseLancer_Markings_Border_Brightness}) {constB=0.3}
   Alpha=SmoothStep(Min=@n71, Max=@n70, Value=Subtract(A=@n74, B=@n75))
  )
  n77 = VectorParameter() {group=PulseLancer_Prototype, parameterName=Proto_BaseColor_1, defaultValue=<rgba 0.1812 0.1912 0.3968 1>}
  n78 = VectorParameter() {group=PulseLancer_Prototype, parameterName=Proto_BaseColor_2, defaultValue=<rgba 0.1812 0.1912 0.3968 1>}
  n79 = NamedRerouteUsage()
  n80 = LinearInterpolate(
   A=@n77.RGB
   B=@n78.RGB
   Alpha=SmoothStep(
    Min=ScalarParameter() {group=PulseLancer_Prototype, parameterName=Proto_Color_Gradient_Min}
    Max=ScalarParameter() {defaultValue=200, group=PulseLancer_Prototype, parameterName=Proto_Color_Gradient_Max}
    Value=ComponentMask(None=@n79) {b=true}
   )
  )
  n81 = LinearInterpolate(A=@n80, B=LinearInterpolate(A=@n76, B=Multiply(A=@n80, B=@n76), Alpha=@n69.A), Alpha=SmoothStep(Min=@n70, Max=@n71, Value=@n21))
  n82 = Multiply(A=@n81, B=ScalarParameter() {defaultValue=0.7, group=PulseLancer_Prototype, parameterName=PulseLancer_LegsGradientBrightness})
  n83 = ComponentMask(None=NamedRerouteUsage()) {b=true}
  n84 = Constant() {r=30}
  n85 = Saturate(None=Divide(A=Subtract(A=@n83, B=@n84), B=Constant() {r=110}) {constB=150})
  n86 = NamedRerouteUsage()
  n87 = NamedRerouteUsage()
  n88 = MaterialFunctionCall(
   Hue Shift Percentage=ScalarParameter() {group=Deimos_BaseColorTint, parameterName=ColorVariants_HueShift_R, primitiveDataIndex=33, sortPriority=29}
   Texture=@n87
  ) {materialFunction=&HueShift}
  n89 = ScalarParameter() {group=Deimos_BaseColorTint, parameterName=ColorVariants_Dessaturate_R, primitiveDataIndex=30, sortPriority=29}
  n90 = ScalarParameter() {group=Deimos_BaseColorTint, parameterName=ColorVariants_Brightness_R, primitiveDataIndex=27, sortPriority=29}
  n91 = TextureSampleParameter2D() {
   group=Deimos_BaseColorTint
   parameterName=ColorVariants_Mask
   samplerType=SAMPLERTYPE_Masks
   sortPriority=28
   texture=&T_WhiteMask
  }
  n92 = MaterialFunctionCall(Color=@n91.RGB) {materialFunction=&SplitComponents}
  n93 = LinearInterpolate(A=@n87, B=Multiply(A=Desaturation(None=@n88.Result, Fraction=@n89), B=Add(A=@n90)), Alpha=@n92.R)
  n94 = ScalarParameter() {group=Deimos_BaseColorTint, parameterName=ColorVariants_HueShift_G, primitiveDataIndex=34, sortPriority=30}
  n95 = Multiply(
   A=Desaturation(
    None=MaterialFunctionCall(Hue Shift Percentage=@n94, Texture=@n87) {materialFunction=&HueShift}.Result
    Fraction=ScalarParameter() {group=Deimos_BaseColorTint, parameterName=ColorVariants_Dessaturate_G, primitiveDataIndex=31, sortPriority=30}
   )
   B=Add(A=ScalarParameter() {group=Deimos_BaseColorTint, parameterName=ColorVariants_Brightness_G, primitiveDataIndex=28, sortPriority=30})
  )
  n96 = Desaturation(
   None=MaterialFunctionCall(
    Hue Shift Percentage=ScalarParameter() {group=Deimos_BaseColorTint, parameterName=ColorVariants_HueShift_B, primitiveDataIndex=35, sortPriority=31}
    Texture=@n87
   ) {materialFunction=&HueShift}.Result
   Fraction=ScalarParameter() {group=Deimos_BaseColorTint, parameterName=ColorVariants_Dessaturate_B, primitiveDataIndex=32, sortPriority=31}
  )
  n97 = Add(A=ScalarParameter() {group=Deimos_BaseColorTint, parameterName=ColorVariants_Brightness_B, primitiveDataIndex=29, sortPriority=31})
  n98 = LinearInterpolate(
   A=NamedRerouteUsage()
   B=LinearInterpolate(A=LinearInterpolate(A=@n93, B=@n95, Alpha=@n92.G), B=Multiply(A=@n96, B=@n97), Alpha=@n92.B)
   Alpha=SmoothStep(
    Min=ScalarParameter() {group=Deimos_BaseColorTint, parameterName=ColorVariants_Gradient_Min}
    Max=ScalarParameter() {group=Deimos_BaseColorTint, parameterName=ColorVariants_Gradient_Max}
    Value=ComponentMask(None=@n86) {b=true}
   )
  )
  n99 = NamedRerouteUsage()
  n100 = StaticBoolParameter() {group=Deimos_BaseColorTint, parameterName=Deimos_BaseColorTint, sortPriority=-10}
  NamedRerouteDeclaration(None=StaticSwitch(True=LinearInterpolate(A=@n82, B=@n81, Alpha=@n85), False=StaticSwitch(True=@n98, False=@n99, Value=@n100), Value=NamedRerouteUsage())) {name=OUT_BaseColor, nodeColor=<rgba 0 1 0.7422 1>}
  NamedRerouteDeclaration(None=@n1.BaseColor) {name=IN_BaseColor, nodeColor=<rgba 1 0.4219 0 1>}
  n101 = NamedRerouteUsage()
  NamedRerouteDeclaration(None=ComponentMask(None=@n101) {b=true, g=true}) {name=UV_PlanarProjSide, nodeColor=<rgba 1 0 0.4922 1>}
  n102 = MaterialFunctionCall(Color=NamedRerouteUsage()) {materialFunction=&SplitComponents}
  n103 = MaterialFunctionCall(X=Abs(None=@n102.R), Y=@n102.G, Z=@n102.G) {materialFunction=&MakeFloat3}
  n104 = VectorParameter() {group=Deimos_Emissive, parameterName=Emissive_Hands_SphereMask_Scale XYZ, sortPriority=51, defaultValue=<rgba 1 1 2.5 1>}
  n105 = VectorParameter() {group=Deimos_Emissive, parameterName=Emissive_Hands_SphereMask_Pos, sortPriority=50, defaultValue=<rgba 70 0 40 1>}
  NamedRerouteDeclaration(None=SphereMask(
    A=Divide(A=@n103.Result, B=@n104.RGB) {constB=1}
    B=@n105.RGB
    Radius=ScalarParameter() {defaultValue=55, group=Deimos_Emissive, parameterName=Emissive_Hands_SphereMask_Radius, sortPriority=51}
    Hardness=ScalarParameter() {defaultValue=0.25, group=Deimos_Emissive, parameterName=Emissive_Hands_SphereMask_Hardness, sortPriority=51}
   ) {attenuationRadius=50, hardnessPercent=0}) {name=M_RangedHands, nodeColor=<rgba 1 0 0.5625 1>}
  n106 = Multiply(A=Time(), B=AppendVector(A=Constant() {r=1}, B=Constant() {r=-20}))
  n107 = Divide(A=Add(A=NamedRerouteUsage(), B=Multiply(A=@n106) {constB=0.2})) {constB=100}
  n108 = Add(A=NamedRerouteUsage(), B=@n106)
  n109 = Constant() {r=150}
  n110 = Multiply(A=TextureSample!(UVs=Divide(A=@n108, B=@n109) {constB=0.2}, Tex=NamedRerouteUsage(), Apply View MipBias) {texture=/Game/Global/Textures/Default/Noise/3xPerlin_M}.R) {constB=0.2}
  NamedRerouteDeclaration(None=TextureSample(UVs=Add(A=@n107, B=@n110), Tex=NamedRerouteUsage()) {texture=&3xPerlin_M}.R) {name=M_NoiseMacro, nodeColor=<rgba 0 0.7656 1 1>}
  n111 = Step(X=ComponentMask(None=NamedRerouteUsage()) {b=true}) {constX=150, constY=155}
  n112 = SphereMask(
   A=Divide(A=NamedRerouteUsage(), B=Constant3Vector() {constant=<rgba 1 1.5 1 0>})
   B=Constant3Vector() {constant=<rgba 0 2 155 1>}
   Radius=Constant() {r=15}
   Hardness=Constant() {r=0.7}
  )
  NamedRerouteDeclaration(None=Saturate(None=Subtract(A=@n111, B=@n112))) {name=M_Proto_horns, nodeColor=<rgba 0 1 0.0156 1>}
  NamedRerouteDeclaration(None=StaticBoolParameter() {defaultValue=true, group=Deimos, parameterName=PrototypeMaterialTricks, sortPriority=-10}) {name=IS_ProtoAdjustments, nodeColor=<rgba 1 0 0.0727 1>}
  n113 = PreSkinnedNormal()
  n114 = Abs(None=VertexInterpolator(VS=@n113).PS)
  n115 = ComponentMask(None=@n114) {b=true}
  n116 = ComponentMask(None=@n114) {g=true, r=true}
  n117 = Subtract(A=@n115, B=@n116)
  n118 = ComponentMask(None=@n117) {r=true}
  n119 = ComponentMask(None=@n117) {g=true}
  n120 = Saturate(None=Subtract(A=@n116, B=@n115))
  n121 = ComponentMask(None=@n120) {r=true}
  n122 = ComponentMask(None=@n120) {g=true}
  n123 = Subtract(A=@n121, B=@n122)
  n124 = Max(A=Saturate(None=Min(A=@n118, B=@n119)), B=Abs(None=@n123))
  NamedRerouteDeclaration(None=Power(
    Base=OneMinus(None=@n124)
    Exp=ScalarParameter() {defaultValue=16, group=PulseLancer_Prototype, parameterName=PulseLancer_DemonMarks_SeamSize}
   )) {name=Tripl_SeamFix_SDF, nodeColor=<rgba 1 0 0.6562 1>}
  NamedRerouteDeclaration(None=DDY(Value=@n56)) {name=DDY, nodeColor=<rgba 1 0 0.7031 1>}
  NamedRerouteDeclaration(None=StaticBoolParameter() {group=Deimos_Emissive, parameterName="Emissive_GlowingHands (Ranged)", sortPriority=-1}) {name=Use_GlowingHands, nodeColor=<rgba 0 0.4609 1 1>}
  NamedRerouteDeclaration(None=StaticBoolParameter() {group=Deimos_Emissive, parameterName="Emissive_Pulsing (Exploding)", sortPriority=-1}) {name=Use_PulsingEmissive, nodeColor=<rgba 1 0.2752 0.4185 1>}
  n125 = Divide(A=NamedRerouteUsage(), B=Constant() {r=15}) {constB=0.2}
  n126 = AppendVector(A=Constant() {r=-0.03}, B=Constant() {r=0.1})
  NamedRerouteDeclaration(None=TextureSample(UVs=Panner(Coordinate=@n125, Speed=@n126), Tex=NamedRerouteUsage()) {texture=&3xPerlin_M}.R) {name=Noise_Scrolling Front, nodeColor=<rgba 0 1 0.6719 1>}
  n127 = CameraVectorWS()
  n128 = Transform(None=Constant3Vector() {constant=<rgba 0 0.2 0 0>}) {transformSourceType=TRANSFORMSOURCE_Local, transformType=TRANSFORM_Camera}
  n129 = DotProduct(A=PixelNormalWS(), B=Normalize(VectorInput=Add(A=@n127, B=@n128)))
  n130 = Add(A=OneMinus(None=Clamp(None=@n129)), B=ScalarParameter() {defaultValue=-0.2, group=Deimos_Emissive, parameterName=Emissive_Rimlight_Size})
  n131 = ScalarParameter() {defaultValue=1, group=Deimos_Emissive, parameterName=Emissive_Rimlight_Contrast}
  n132 = ScalarParameter() {defaultValue=80, group=Deimos_Emissive, parameterName=Emissive_Rimlight_Gradient_Min}
  n133 = ScalarParameter() {defaultValue=200, group=Deimos_Emissive, parameterName=Emissive_Rimlight_Gradient_Max}
  n134 = ComponentMask(None=NamedRerouteUsage()) {b=true}
  n135 = Multiply(
   A=Multiply(
    A=MaterialFunctionCall(In=@n130, Contrast=@n131) {materialFunction=&CheapContrast}.Result
    B=SmoothStep(Min=@n132, Max=@n133, Value=@n134)
   )
   B=VectorParameter() {group=Deimos_Emissive, parameterName=Emissive_RimLight_Color, defaultValue=<rgba 0.6437 0.1842 1 1>}.RGB
  )
  NamedRerouteDeclaration(None=@n135) {name=Emiss_RimLIght, nodeColor=<rgba 0 1 0.1328 1>}
  n136 = VectorParameter() {group=Deimos_Emissive, parameterName=Emissive_Pulsing_SphereMask_Scale XYZ, defaultValue=<rgba 1 1 1 1>}
  n137 = Divide(A=NamedRerouteUsage(), B=@n136.RGB) {constB=1}
  n138 = VectorParameter() {group=Deimos_Emissive, parameterName=Emissive_Pulsing_SphereMask_Pos, defaultValue=<rgba 0 10 100 1>}
  n139 = ScalarParameter() {defaultValue=40, group=Deimos_Emissive, parameterName=Emissive_Pulsing_SphereMask_Radius}
  n140 = ScalarParameter() {defaultValue=0.1, group=Deimos_Emissive, parameterName=Emissive_Pulsing_SphereMask_Hardness}
  NamedRerouteDeclaration(None=Power(
    Base=SphereMask(A=@n137, B=@n138.RGB, Radius=@n139, Hardness=@n140)
    Exp=ScalarParameter() {defaultValue=4, group=Deimos_Emissive, parameterName=Emissive_Pulsing_SphereMask_Power}
   )) {name=M_PulsingSphereMask, nodeColor=<rgba 0 0.7656 1 1>}
  NamedRerouteDeclaration(None=Ceil(None=@n123)) {name=Flat R_1, nodeColor=<rgba 1 0 0.8906 1>}
  NamedRerouteDeclaration(None=Ceil(None=Subtract(A=@n122, B=@n121))) {name=Flat G_1, nodeColor=<rgba 0 1 0.7891 1>}
  n141 = MaterialFunctionCall(Color=@n101) {materialFunction=&SplitComponents}
  NamedRerouteDeclaration(None=AppendVector(A=Abs(None=@n141.R), B=@n141.B)) {name=UV_PlanarProjFront_Symetric, nodeColor=<rgba 1 0 0.4922 1>}
  NamedRerouteDeclaration(None=TextureObject() {samplerType=SAMPLERTYPE_Grayscale, texture=/Game/Global/Textures/Default/Noise/LowResBlurredNoiseNormalized}) {name=TO_PerlinNoise, nodeColor=<rgba 0 0.4844 1 1>}
  n142 = Divide(A=NamedRerouteUsage(), B=ScalarParameter() {defaultValue=9, group=Deimos_Emissive, parameterName=Emissive_Pulsing_Pattern_Scale}) {constB=10}
  n143 = MaterialFunctionCall() {materialFunction=&CameraDirectionVector}
  n144 = Constant3Vector() {constant=<rgba -1 -1 1 0>}
  n145 = ComponentMask(None=Multiply(
    A=Multiply(A=@n143.Result, B=@n144)
    B=ScalarParameter() {defaultValue=1, group=Deimos_Emissive, parameterName=Emissive_Pulsing_Pattern_BumpOffset}
   ) {constB=-1}) {b=true, r=true}
  n146 = TextureObjectParameter() {group=Deimos_Emissive, parameterName=Emissive_Pulsing_Pattern, texture=/Game/Global/Textures/Default/Noise/T_SoftCells_Negative}
  NamedRerouteDeclaration(None=TextureSample(UVs=Add(A=@n142, B=@n145), Tex=@n146) {texture=&T_SoftCells_Negative}.R) {name=T_SoftCell, nodeColor=<rgba 0 0.4844 1 1>}
  n147 = Divide(A=NamedRerouteUsage(), B=Constant() {r=5})
  n148 = NamedRerouteUsage()
  NamedRerouteDeclaration(None=Max(A=TextureSample(UVs=@n147, Tex=@n148).R, B=NamedRerouteUsage())) {name=T_Detail, nodeColor=<rgba 0.8047 1 0 1>}
  NamedRerouteDeclaration(None=ComponentMask(None=@n101) {b=true, r=true}) {name=UV_PlanarProjFront, nodeColor=<rgba 1 0 0.4922 1>}
  n149 = NamedRerouteUsage()
  n150 = Multiply(A=@n149, B=Constant3Vector() {constant=<rgba 1 1 -1 0>})
  n151 = ComponentMask(None=@n150) {g=true, r=true}
  n152 = ComponentMask(None=@n150) {b=true, r=true}
  n153 = NamedRerouteUsage()
  NamedRerouteDeclaration(None=LinearInterpolate(A=LinearInterpolate(A=@n151, B=@n152, Alpha=@n153), B=ComponentMask(None=@n150) {b=true, g=true}, Alpha=NamedRerouteUsage())) {name=TriplanarUV, nodeColor=<rgba 1 0 0.9141 1>}
  n154 = Multiply(A=NamedRerouteUsage()) {constB=5}
  n155 = SmoothStep(
   Min=ScalarParameter() {defaultValue=150, group=PulseLancer_Prototype, parameterName=Proto_Marks_Min}
   Max=ScalarParameter() {defaultValue=120, group=PulseLancer_Prototype, parameterName=Proto_Marks_Max}
   Value=ComponentMask(None=NamedRerouteUsage()) {b=true}
  ) {constMax=0, constMin=100}
  n156 = StaticSwitch(True=LinearInterpolate(A=@n155, Alpha=Saturate(None=@n154)) {constB=0}, False=@n155, Value=NamedRerouteUsage())
  NamedRerouteDeclaration(None=@n156) {name=Remove Demon Marks, nodeColor=<rgba 0 0.5312 1 1>}
  NamedRerouteUsage()
  NamedRerouteUsage()
  Reroute(None=@n135)
  n157 = Constant3Vector() {constant=<rgba 0 0 1 0>}
  n158 = Constant() {r=0.25}
  n159 = Constant3Vector()
  RotateAboutAxis(NormalizedRotationAxis=@n157, RotationAngle=@n158, PivotPoint=@n159, Position=@n113)
  RotateAboutAxis(NormalizedRotationAxis=@n157, RotationAngle=@n158, PivotPoint=@n159, Position=@n149)
  n160 = Multiply(A=NamedRerouteUsage()) {constB=5}
  StaticSwitch(True=LinearInterpolate(A=@n156, Alpha=Saturate(None=@n160)) {constB=0}, False=@n156, Value=NamedRerouteUsage())
  Transform(None=@n114) {transformSourceType=TRANSFORMSOURCE_Local, transformType=TRANSFORM_Tangent}
 ]}
