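# Deimos enemy material for the PulseLancer fiend, a lit skeletal mesh surface with mesh deformer support and purple flesh physics.
# A PulseLancer texture function supplies material attributes to a PulseLancer look-dev function.
# That output passes through the shared Deimos enemy final attribute function, the Creative move tool function, then a cinematic character post-FX function.
# All shading logic lives in these chained material functions.
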Material /PulseLancer_SmashAI/Assets/Materials/M_DeimosEnemy_PulseLancer
graph {
 properties {
  bUseMaterialAttributes=true
  bUsedWithMeshDeformer=true
  bUsedWithSkeletalMesh=true
  physMaterial=/Game/FortressPhysicalMaterials/Purple_Flesh
  pixelDepthOffsetMode=PDOM_Legacy
  usedShadingModels=Default Lit
 }
 nodes [
  n1 = MaterialFunctionCall!(Result=MaterialFunctionCall() {materialFunction=&/MF_DeimosEnemy_Pulselancer_Textures}.MakeMaterialAttributes) {materialFunction=&/MF_FiendLookDev_PulseLancer}
  MaterialFunctionCall(In=MaterialFunctionCall(MA In=MaterialFunctionCall!(Material=@n1.SetMaterialAttributes) {materialFunction=/Game/Characters/Enemies/MaterialAttributes/MF_DeimosEnemyFinalMaterialAttribute}.Result) {materialFunction=&MF_CreativeMoveTool}.MA Out) {materialFunction=&CINE_MF_CharPostFX}
 ]
}
