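# Additive unlit material, named for a shield break, that computes a vertex-style offset relative to the Fortnite storm safe zone.
# The offset shifts radially from the zone center using actor and camera direction, with a distance-driven hide height.
# Sine and panning aura-noise texture add wobble.
# A dimmed red-to-green Fresnel supplies color, while named reroutes define storm radius, camera-outside-zone and player-proximity masks.
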Material /PulseLancer_SmashAI/Assets/Materials/M_ShieldBreak
graph {
 properties {
  allowTranslucentCustomDepthWrites=true
  bAutomaticallySetUsageInEditor=false
  bUseTranslucencyVertexFog=false
  blendMode=BLEND_Additive
  shadingModel=MSM_Unlit
  usedShadingModels=Unlit
 }
 nodes [
  n1 = ScalarParameter() {defaultValue=0.3393, parameterName=Noise Size 2}
  n2 = Time()
  n3 = Sine(None=Add(A=Multiply(A=@n1, B=@n2), B=LocalPosition().XYZ))
  n4 = ScalarParameter() {defaultValue=20, parameterName=Noise Intensity 2}
  n5 = Add(A=Divide(A=Time()) {constB=100}, B=TextureCoordinate())
  n6 = ScalarParameter() {defaultValue=1.1737, parameterName=Noise Size}
  n7 = Add(A=TextureSample(UVs=Multiply(A=@n5, B=@n6)) {texture=/PulseLancer_SmashAI/Assets/Textures/T_AuraNoise}.G) {constB=-0.5}
  n8 = ScalarParameter() {defaultValue=1.8396, parameterName=Noise Intensity}
  n9 = ComponentMask(None=ActorPositionWS()) {g=true, r=true}
  n10 = CollectionParameter() {collection=&FortniteMaterialParameters, parameterName=AthenaSafeZoneCenter}
  n11 = ComponentMask(None=@n10) {g=true, r=true}
  n12 = Add(A=Multiply(A=@n9) {constB=-1}, B=@n11)
  n13 = Normalize(VectorInput=Subtract(A=@n9, B=@n11))
  n14 = Distance(A=@n9, B=@n11)
  n15 = NamedRerouteUsage()
  n16 = Max(A=Abs(None=Subtract(A=@n14, B=@n15)), B=ScalarParameter() {defaultValue=280, parameterName=Sphere Radius})
  n17 = Normalize(VectorInput=ComponentMask(None=ActorPositionWS() {originType=CameraRelative}) {g=true, r=true})
  n18 = Multiply(A=@n16, B=Sign(None=DotProduct(A=@n13, B=@n17))) {constB=-1}
  n19 = NamedRerouteUsage()
  n20 = Multiply(A=@n13, B=Add(A=LinearInterpolate(A=@n18, B=@n16, Alpha=@n19), B=NamedRerouteUsage()))
  n21 = ScalarParameter() {defaultValue=-8767.4277, parameterName=Hide Height}
  n22 = Multiply(A=@n15) {constB=0.5}
  n23 = Saturate(None=Multiply(A=Divide(A=@n14, B=@n22), B=NamedRerouteUsage()))
  Add(
   A=Add(A=Multiply(A=@n3, B=@n4), B=Multiply(A=@n7, B=@n8))
   B=AppendVector(A=Add(A=@n12, B=@n20), B=LinearInterpolate(A=@n21, Alpha=@n23) {constB=0})
  )
  CollectionParameter() {collection=&FortniteMaterialParameters, parameterName=AthenaSafeZoneCenter}
  Constant()
  Constant()
  n24 = Constant3Vector() {constant=<rgba 3 0 0 0>}
  n25 = Constant3Vector() {constant=<rgba 0 1 0 0>}
  Multiply(A=Multiply(A=Fresnel(), B=LinearInterpolate(A=@n24, B=@n25))) {constB=0.1}
  n26 = CollectionParameter() {collection=&FortniteMaterialParameters, parameterName=AthenaSafeZoneRadius}
  NamedRerouteDeclaration(None=@n26) {name=Radius Storm Shield, nodeColor=<rgba 0 0.25 1 1>}
  n27 = Subtract(
   A=CameraPositionWS()
   B=ComponentMask(None=CollectionParameter() {collection=/Game/Packages/Fortress_SharedMaterials/GlobalMaterialParameters/DirectSafeZoneMPC, parameterName=SafeLocation}) {b=true, g=true, r=true}
  )
  n28 = NamedRerouteUsage()
  n29 = Add(A=Distance(A=ComponentMask(None=@n27) {g=true, r=true}, B=Constant()), B=Multiply(A=@n28) {constB=-0.5})
  n30 = Round(None=Saturate(None=Divide(A=@n29, B=@n28)))
  NamedRerouteDeclaration(None=Saturate(None=OneMinus(None=@n30))) {name=Outzone, nodeColor=<rgba 1 0.1641 0 1>}
  n31 = Subtract(A=CameraPositionWS(), B=ActorPositionWS())
  n32 = Divide(
   A=Distance(A=ComponentMask(None=@n31) {g=true, r=true}, B=Constant())
   B=ScalarParameter() {defaultValue=650, parameterName=Player Distance Mask}
  )
  NamedRerouteDeclaration(None=Saturate(None=Multiply(A=@n32, B=@n32))) {name=Player Zone, nodeColor=<rgba 1 0.8672 0 1>}
 ]
}
