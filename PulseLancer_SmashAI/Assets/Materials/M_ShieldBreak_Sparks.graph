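# Unlit translucent spark material, named for shield breaks, with graph logic built around the Fortnite storm safe-zone barrier.
# Camera-facing sprite offsets use sphere-mask size curves and a platform switch that zeroes size on Switch, Android and PC performance mode.
# Named reroutes supply looping ages staggered by vertex color, curl-noise drift, velocity, camera-distance masks and a push relative to the barrier.
# A soft radial hot center brightens the tint color.
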Material /PulseLancer_SmashAI/Assets/Materials/M_ShieldBreak_Sparks
graph {
 properties {bAutomaticallySetUsageInEditor=false, blendMode=BLEND_Translucent, shadingModel=MSM_Unlit, twoSided=true, usedShadingModels=Unlit}
 nodes [
  ActorPositionWS()
  n1 = WorldPosition()
  n2 = TextureCoordinate()
  n3 = AppendVector(A=ScalarParameter() {defaultValue=35, parameterName=X size}, B=ScalarParameter() {defaultValue=90, parameterName=Y Size})
  n4 = NamedRerouteUsage()
  n5 = ScalarParameter() {defaultValue=0.5, parameterName=Scale Curve Start}
  n6 = ScalarParameter() {defaultValue=0.5, parameterName=Scale Curve Size}
  n7 = ScalarParameter() {parameterName=Scale Curve Hardness}
  n8 = Multiply(A=LinearInterpolate(B=@n3, Alpha=SphereMask(A=@n4, B=@n5, Radius=@n6, Hardness=@n7)), B=NamedRerouteUsage())
  n9 = OneMinus(None=NamedRerouteUsage())
  n10 = Constant() {r=1}
  n11 = Constant()
  n12 = Multiply(
   A=Multiply(A=Multiply(A=@n8, B=@n9), B=NamedRerouteUsage())
   B=MaterialFunctionCall!("Xb1,PS4"=@n10, "PC-High&&Epic,XSX,PS5"=@n10, "PC-Perf,Android"=@n11, Switch=@n11, "PC-Low(NotPerf)&&Medium"=@n10) {materialFunction=/Game/Effects/Niagara/MasterMaterials/Functions/Common/PlatformSwitch_3D}.Result
  )
  Add(
   A=Add(
    A=MaterialFunctionCall!(Center Location=@n1.XYZ, 0-1 UVs=@n2, XY scale=@n12, OverridePivot, "Normals (Optional)") {materialFunction=&Sprite}.World position offset
    B=NamedRerouteUsage()
   )
   B=NamedRerouteUsage()
  )
  n13 = ScalarParameter() {defaultValue=0.3393, parameterName=Noise Size 2}
  n14 = Time()
  n15 = Sine(None=Add(A=Multiply(A=@n13, B=@n14), B=LocalPosition().XYZ))
  n16 = ScalarParameter() {defaultValue=33.0427, parameterName=Noise Intensity 2}
  n17 = Add(A=Divide(A=Time()) {constB=100}, B=TextureCoordinate())
  n18 = ScalarParameter() {defaultValue=1.1737, parameterName=Noise Size}
  n19 = Add(A=TextureSample(UVs=Multiply(A=@n17, B=@n18)) {texture=/PulseLancer_SmashAI/Assets/Textures/T_AuraNoise}.G) {constB=-0.5}
  n20 = ScalarParameter() {defaultValue=-75.2714, parameterName=Noise Intensity}
  Add(A=Add(A=Multiply(A=@n15, B=@n16), B=Multiply(A=@n19, B=@n20)))
  Add()
  n21 = TextureCoordinate()
  n22 = ScalarParameter() {defaultValue=0.5, parameterName=Param_2}
  n23 = ScalarParameter() {defaultValue=0.25, parameterName=hot center radius}
  n24 = ScalarParameter() {parameterName=hot center size}
  Add(
   A=Multiply(
    A=SphereMask(A=@n21, B=@n22, Radius=@n23, Hardness=@n24)
    B=ScalarParameter() {defaultValue=0.5, parameterName=Center brightness contribution}
   )
   B=VectorParameter() {parameterName=Color, defaultValue=<rgba 1.5 0 0.116 1>}.RGB
  )
  n25 = ComponentMask(None=ActorPositionWS()) {g=true, r=true}
  n26 = Multiply(A=@n25) {constB=-1}
  n27 = CollectionParameter() {collection=&FortniteMaterialParameters, parameterName=AthenaSafeZoneCenter}
  n28 = ComponentMask(None=@n27) {g=true, r=true}
  n29 = Normalize(VectorInput=Subtract(A=@n25, B=@n28))
  n30 = Distance(A=@n25, B=@n28)
  n31 = NamedRerouteUsage()
  n32 = Max(A=Abs(None=Subtract(A=@n30, B=@n31)), B=ScalarParameter() {defaultValue=300, parameterName=Sphere Radius})
  n33 = ComponentMask(None=ActorPositionWS() {originType=CameraRelative}) {g=true, r=true}
  n34 = Sign(None=DotProduct(A=@n29, B=Normalize(VectorInput=@n33)))
  n35 = Add(A=LinearInterpolate(A=Multiply(A=@n32, B=@n34) {constB=-1}, B=@n32, Alpha=NamedRerouteUsage()), B=NamedRerouteUsage())
  n36 = Multiply(A=Divide(A=@n30, B=Multiply(A=@n31) {constB=0.5}), B=NamedRerouteUsage())
  AppendVector(
   A=Add(A=Add(A=@n26, B=@n28), B=Multiply(A=@n29, B=@n35))
   B=LinearInterpolate(A=ScalarParameter() {defaultValue=-8767.4277, parameterName=Hide Height}, Alpha=Saturate(None=@n36)) {constB=0}
  )
  CollectionParameter() {collection=/Game/Athena/SafeZone/MPC_SafeZoneClouds, parameterName=SafeZoneScale}
  n37 = ComponentMask(None=ActorPositionWS()) {g=true, r=true}
  n38 = CollectionParameter() {collection=&FortniteMaterialParameters, parameterName=AthenaSafeZoneCenter}
  n39 = ComponentMask(None=@n38) {g=true, r=true}
  n40 = Distance(A=@n37, B=@n39)
  n41 = NamedRerouteUsage()
  n42 = Max(A=Abs(None=Subtract(A=@n40, B=@n41)), B=ScalarParameter() {defaultValue=300, parameterName=Sphere Radius})
  n43 = Normalize(VectorInput=Subtract(A=@n37, B=@n39))
  n44 = ActorPositionWS() {originType=CameraRelative}
  n45 = DotProduct(A=@n43, B=Normalize(VectorInput=ComponentMask(None=@n44) {g=true, r=true}))
  LinearInterpolate(A=Multiply(A=@n42, B=Sign(None=@n45)) {constB=-1}, B=@n42, Alpha=NamedRerouteUsage())
  n46 = Constant()
  MaterialFunctionCall(
   Input=Saturate(None=ScalarParameter() {defaultValue=1, parameterName=Param_1})
   Input Low=@n46
   Input High=Constant() {r=1}
   Target Low=@n46
   Target High=Constant() {r=185185}
  ) {materialFunction=&RemapValueRange}
  n47 = LocalPosition() {localOrigin=InstancePreSkinning}
  n48 = Add(A=WorldPosition().XYZ, B=NamedRerouteUsage())
  MaterialFunctionCall(
   Input=Length(None=Subtract(A=@n47.XYZ, B=@n48)) {is=class MaterialExpressionLength}
   Input High=ScalarParameter() {defaultValue=1750, parameterName=Param_8}
   Target High=Constant() {r=1}
  ) {materialFunction=&RemapValueRange}
  Multiply(B=ComponentMask(None=TextureCoordinate()) {r=true})
  Multiply(A=@n3)
  Multiply(A=SphereMask(A=TextureCoordinate(), B=Constant() {r=0.5}, Radius=Constant() {r=0.5}, Hardness=Constant()), B=NamedRerouteUsage())
  n49 = Add(A=Time(), B=VertexColor().G)
  n50 = ScalarParameter() {defaultValue=1, parameterName=Age loop duration}
  NamedRerouteDeclaration(None=Divide(A=Fmod(A=@n49, B=@n50), B=@n50)) {name=NOrmalized Age Scale, nodeColor=<rgba 0 1 0.2734 1>}
  n51 = Add(A=Add(A=Time(), B=NamedRerouteUsage()), B=VertexColor().G)
  n52 = ScalarParameter() {defaultValue=1, parameterName=Curl Duration}
  NamedRerouteDeclaration(None=Divide(A=Fmod(A=@n51, B=@n52), B=@n52)) {name=NOrmalized Age CUrl, nodeColor=<rgba 0 1 0.2734 1>}
  n53 = MaterialFunctionCall(Float3=ActorPositionWS()) {materialFunction=&BreakOutFloat3Components}
  n54 = Add(A=@n53.R, B=@n53.R)
  NamedRerouteDeclaration(None=Multiply(A=Add(A=@n54, B=@n53.B), B=ScalarParameter() {defaultValue=49854799872, parameterName=Seed})) {name=Random, nodeColor=<rgba 0 0.0156 1 1>}
  n55 = Time()
  n56 = VectorParameter() {parameterName=Param_2, defaultValue=<rgba 0 0 10 0>}
  n57 = Multiply(A=Add(A=LocalPosition().XYZ, B=Multiply(A=@n55, B=@n56.RGB)), B=ScalarParameter() {defaultValue=0.0012, parameterName=Noise Scale Size})
  NamedRerouteDeclaration(None=Multiply(
    A=TextureSample(UVs=@n57) {samplerType=SAMPLERTYPE_LinearColor, texture=/Game/Effects/FogVolume/VT_CurlNoise_LOW}.RGB
    B=VectorParameter() {parameterName=Noise Scale Amp, defaultValue=<rgba 25 50 50 0>}.RGB
   )) {name=Curl Noise, nodeColor=<rgba 0.2891 1 0 1>}
  n58 = Multiply(A=VectorParameter() {defaultValue=<rgba 0 0 375 0>}.RGB, B=NamedRerouteUsage())
  n59 = NamedRerouteUsage()
  n60 = ActorPositionWS()
  n61 = WorldPosition()
  n62 = Multiply(A=Normalize(VectorInput=Subtract(A=@n60, B=@n61.XYZ)), B=NamedRerouteUsage())
  n63 = ScalarParameter() {defaultValue=-50, parameterName=Velocity Scale}
  NamedRerouteDeclaration(None=Add(A=Add(A=@n58, B=@n59), B=Multiply(A=@n62, B=@n63))) {name=Velocity, nodeColor=<rgba 1 0 0.5625 1>}
  n64 = Add(A=Multiply(A=@n37) {constB=-1}, B=@n39)
  n65 = Multiply(A=@n43, B=Add(A=@n42, B=NamedRerouteUsage()))
  n66 = ScalarParameter() {defaultValue=-8767.4277, parameterName=Hide Height}
  n67 = Multiply(A=@n41) {constB=0.5}
  n68 = Saturate(None=Multiply(A=Divide(A=@n40, B=@n67), B=NamedRerouteUsage()))
  NamedRerouteDeclaration(None=AppendVector(A=Add(A=@n64, B=@n65), B=LinearInterpolate(A=@n66, Alpha=@n68) {constB=0})) {name=Avoid Barrier, nodeColor=<rgba 0.1484 0 1 1>}
  n69 = CollectionParameter() {collection=&FortniteMaterialParameters, parameterName=AthenaSafeZoneRadius}
  NamedRerouteDeclaration(None=@n69) {name=Radius Storm Shield, nodeColor=<rgba 0 0.25 1 1>}
  n70 = CollectionParameter() {collection=&FortniteMaterialParameters, parameterName=AthenaSafeZoneCenter}
  n71 = Subtract(A=CameraPositionWS(), B=ComponentMask(None=@n70) {b=true, g=true, r=true})
  n72 = NamedRerouteUsage()
  n73 = Add(A=Distance(A=ComponentMask(None=@n71) {g=true, r=true}, B=Constant()), B=Multiply(A=@n72) {constB=-0.5})
  n74 = Round(None=Saturate(None=Divide(A=@n73, B=@n72)))
  NamedRerouteDeclaration(None=Saturate(None=OneMinus(None=@n74))) {name=Outzone, nodeColor=<rgba 1 0.1641 0 1>}
  n75 = Subtract(A=CameraPositionWS(), B=ActorPositionWS())
  n76 = Divide(
   A=Distance(A=ComponentMask(None=@n75) {g=true, r=true}, B=Constant())
   B=ScalarParameter() {defaultValue=650, parameterName=Player Distance Mask}
  )
  NamedRerouteDeclaration(None=Saturate(None=Multiply(A=@n76, B=@n76))) {name=Player Zone, nodeColor=<rgba 1 0.8672 0 1>}
  NamedRerouteDeclaration(None=SphereMask(
    A=ActorPositionWS()
    B=CameraPositionWS()
    Radius=ScalarParameter() {defaultValue=7000, parameterName=Camera Scale}
    Hardness=ScalarParameter() {parameterName=Camera Scale Hardness}
   )) {name=Camera Distance, nodeColor=<rgba 1 0 0.9844 1>}
  n77 = CollectionParameter() {collection=&MPC_SafeZoneClouds, parameterName=SafeZoneLoc}
  NamedRerouteDeclaration(None=SphereMask(A=WorldPosition().XYZ, B=ComponentMask(None=@n77) {b=true, g=true, r=true}, Radius=NamedRerouteUsage())) {name=Barrier Hide, nodeColor=<rgba 1 0.0469 0 1>}
  NamedRerouteUsage()
  ScalarParameter() {defaultValue=175, parameterName=Param_7}
  ScalarParameter() {defaultValue=5136.4092, parameterName=Param_9}
  ScalarParameter() {defaultValue=90}
  ScalarParameter() {defaultValue=0.027, parameterName=debug}
  VectorParameter() {defaultValue=<rgba 0 0 375 0>}
  VertexInterpolator(VS=@n47.XYZ)
  WorldPosition()
  WorldPosition()
 ]
}
